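# Niagara system for the Zero Point updraft area: GPU particles spiral and rise over a wide ring-shaped region.
# Flipbook sand sprites, mesh drops and landscape-sampled ground motes are driven upward by vortex and gravity or linear forces.
# A secondary drop emitter spawns from the first drops through particle attribute reads.
# Time-of-day inputs tint sand and drops, camera-distance remaps adjust size and alpha, and a region mask shapes ground-mote lift and speed.
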NiagaraSystem /Game/Environments/Hera/FX/ZeroPoint/NS_ZeroPoint_Updraft_Area
dependencies {
 usedDataInterfaces [
  &NiagaraDataInterfaceCurve
  &NiagaraDataInterfaceParticleRead
  &NiagaraDataInterfaceVectorCurve
  &NiagaraDataInterfaceColorCurve
  &NiagaraDataInterfaceCamera
  &NiagaraDataInterfaceLandscape
  &NiagaraDataInterfaceSpriteRendererInfo
 ]
 usedDynamicInputs [
  &Vector2DFromFloat
  &MakeLinearColor
  &Sine
  &FNTimeOfDayColor
 ]
 usedModules [
  &self:SystemSpawnScript.SystemScriptSource.SystemScriptGraph.NiagaraNodeAssignment_0.SetVariables_#1
  &SystemState
  &self:SystemSpawnScript.SystemScriptSource.SystemScriptGraph.NiagaraNodeAssignment_2.SetVariables_#2
  &self:SystemSpawnScript.SystemScriptSource.SystemScriptGraph.NiagaraNodeAssignment_1.SetVariables_#3
  &self:SystemSpawnScript.SystemScriptSource.SystemScriptGraph.NiagaraNodeAssignment_3.SetVariables_#4
 ]
}
emitters {
 BlowingSandFlipbook {
  simTarget=GPUComputeSim
  emitterUpdateScript [
   &EmitterState(Life Cycle Mode=::NewEnumerator1, Loop Duration=2, UseLoopDelay=false)
   &SpawnRate(SpawnRate=15, Use Spawn Probability=false)
  ]
  particleSpawnScript [
   &InitializeParticle(
    Color=&ColorFromCurve(
     ColorCurve=DataInterface() {propertyValues {
       AlphaCurve {keys [<at 0 1>,<at 1 1>]}
       BlueCurve {keys [
         {
          arriveTangent=-0.1803
          arriveTangentWeight=0.2479
          interpMode=RCIM_Cubic
          leaveTangent=-0.1803
          leaveTangentWeight=0.2479
          tangentMode=RCTM_User
          tangentWeightMode=RCTWM_WeightedBoth
          time=0.0006
          value=0.8177
         }
         {
          arriveTangentWeight=0.3333
          interpMode=RCIM_Cubic
          leaveTangentWeight=0.3333
          tangentMode=RCTM_User
          tangentWeightMode=RCTWM_WeightedBoth
          time=0.375
          value=0.6927
         }
         {
          arriveTangentWeight=0.3333
          interpMode=RCIM_Cubic
          leaveTangentWeight=0.3333
          tangentMode=RCTM_User
          tangentWeightMode=RCTWM_WeightedBoth
          time=0.5
          value=0.6927
         }
         {
          arriveTangentWeight=0.3333
          interpMode=RCIM_Cubic
          leaveTangentWeight=0.3333
          tangentMode=RCTM_User
          tangentWeightMode=RCTWM_WeightedBoth
          time=0.625
          value=0.7448
         }
         {
          arriveTangent=0.1498
          arriveTangentWeight=0.3789
          interpMode=RCIM_Cubic
          leaveTangent=-1.2056
          leaveTangentWeight=0.3115
          tangentMode=RCTM_Break
          tangentWeightMode=RCTWM_WeightedBoth
          time=0.75
          value=0.6134
         }
         {
          arriveTangent=-0.3551
          arriveTangentWeight=0.2296
          interpMode=RCIM_Cubic
          leaveTangent=-1.0064
          leaveTangentWeight=0.2105
          tangentMode=RCTM_Break
          tangentWeightMode=RCTWM_WeightedBoth
          time=0.898
          value=0.1229
         }
         {arriveTangent=7.8868, interpMode=RCIM_Cubic, leaveTangent=7.8868, time=0.9114, value=0.1303}
         {
          arriveTangent=0.0005
          arriveTangentWeight=0.1844
          interpMode=RCIM_Cubic
          leaveTangent=0.0005
          leaveTangentWeight=0.1844
          tangentMode=RCTM_User
          tangentWeightMode=RCTWM_WeightedBoth
          time=1
          value=0.9271
         }
        ]}
       GreenCurve {keys [
         {
          arriveTangent=-0.1803
          arriveTangentWeight=0.2479
          interpMode=RCIM_Cubic
          leaveTangent=-0.1803
          leaveTangentWeight=0.2479
          tangentMode=RCTM_User
          tangentWeightMode=RCTWM_WeightedBoth
          time=0.0006
          value=0.6644
         }
         {
          arriveTangentWeight=0.3333
          interpMode=RCIM_Cubic
          leaveTangentWeight=0.3333
          tangentMode=RCTM_User
          tangentWeightMode=RCTWM_WeightedBoth
          time=0.375
          value=0.617
         }
         {
          arriveTangentWeight=0.3333
          interpMode=RCIM_Cubic
          leaveTangentWeight=0.3333
          tangentMode=RCTM_User
          tangentWeightMode=RCTWM_WeightedBoth
          time=0.5
          value=0.5885
         }
         {
          arriveTangentWeight=0.3333
          interpMode=RCIM_Cubic
          leaveTangentWeight=0.3333
          tangentMode=RCTM_User
          tangentWeightMode=RCTWM_WeightedBoth
          time=0.625
          value=0.6743
         }
         {
          arriveTangent=0.3024
          arriveTangentWeight=0.3546
          interpMode=RCIM_Cubic
          leaveTangent=-3.1217
          leaveTangentWeight=0.4808
          tangentMode=RCTM_Break
          tangentWeightMode=RCTWM_WeightedBoth
          time=0.75
          value=0.5991
         }
         {
          arriveTangent=-1.3022
          arriveTangentWeight=0.1327
          interpMode=RCIM_Cubic
          leaveTangent=-2.6654
          leaveTangentWeight=0.1837
          tangentMode=RCTM_Break
          tangentWeightMode=RCTWM_WeightedBoth
          time=0.898
          value=0.1417
         }
         {arriveTangent=5.997, interpMode=RCIM_Cubic, leaveTangent=5.997, time=0.9114, value=0.0806}
         {
          arriveTangent=0.6573
          arriveTangentWeight=0.3221
          interpMode=RCIM_Cubic
          leaveTangent=0.6573
          leaveTangentWeight=0.3221
          tangentMode=RCTM_User
          tangentWeightMode=RCTWM_WeightedBoth
          time=1
          value=0.7533
         }
        ]}
       RedCurve {keys [
         {
          arriveTangent=-0.1803
          arriveTangentWeight=0.2479
          interpMode=RCIM_Cubic
          leaveTangent=-0.1803
          leaveTangentWeight=0.2479
          tangentMode=RCTM_User
          tangentWeightMode=RCTWM_WeightedBoth
          time=0.0006
          value=0.4089
         }
         {
          arriveTangentWeight=0.3333
          interpMode=RCIM_Cubic
          leaveTangentWeight=0.3333
          tangentMode=RCTM_User
          tangentWeightMode=RCTWM_WeightedBoth
          time=0.375
          value=0.5448
         }
         {
          arriveTangentWeight=0.3333
          interpMode=RCIM_Cubic
          leaveTangentWeight=0.3333
          tangentMode=RCTM_User
          tangentWeightMode=RCTWM_WeightedBoth
          time=0.5
          value=0.4149
         }
         {
          arriveTangentWeight=0.3333
          interpMode=RCIM_Cubic
          leaveTangentWeight=0.3333
          tangentMode=RCTM_User
          tangentWeightMode=RCTWM_WeightedBoth
          time=0.625
          value=0.6013
         }
         {
          arriveTangent=0.3024
          arriveTangentWeight=0.3546
          interpMode=RCIM_Cubic
          leaveTangent=0.3024
          leaveTangentWeight=0.3546
          tangentMode=RCTM_User
          tangentWeightMode=RCTWM_WeightedBoth
          time=0.75
          value=0.72
         }
         {
          arriveTangent=-3.5052
          arriveTangentWeight=0.2417
          interpMode=RCIM_Cubic
          leaveTangent=-12.9556
          leaveTangentWeight=0.288
          tangentMode=RCTM_Break
          tangentWeightMode=RCTWM_WeightedBoth
          time=0.898
          value=0.474
         }
         {arriveTangent=-0.1022, interpMode=RCIM_Cubic, leaveTangent=-0.1022, time=0.9114, value=0.0219}
         {arriveTangentWeight=0.3333, interpMode=RCIM_Cubic, tangentMode=RCTM_User, tangentWeightMode=RCTWM_WeightedBoth, time=1, value=0.4635}
        ]}
      }}
     CurveIndex=Linked() {linkedVariable {name=NPC.FortniteNPC.TimeOfDay_0-1, type=NiagaraFloat}}
    )
    Lifetime Max=3
    Lifetime Min=2
    Lifetime Mode=::NewEnumerator1
    Mass Mode=::NewEnumerator3
    Sprite Rotation Mode=::NewEnumerator2
    Sprite Size Max=<600 1200>
    Sprite Size Min=<300 500>
    Sprite Size Mode=::NewEnumerator2
    Sprite UV Mode=::NewEnumerator1
    UsePositionOffset=false
    Write Material Random=true
   )
   &ShapeLocation(
    Handle Radius=250
    Large Radius=3000
    MaskSpawnGroup=false
    Offset=<0 0 -400>
    Offset Mode=::NewEnumerator0
    Override Seed=false
    Shape Primitive=::NewEnumerator3
    UseOwnerScale=true
   ) {module=ShapeLocation001}
   &AddVelocity(Velocity=<0 0 2500>)
   &SubUVAnimation(End Frame=3, SubUV Animation Mode:ENiagara_SubUVLookupMode=::NewEnumerator2)
  ]
  particleUpdateScript [
   &ParticleState()
   /SetVariables(Particles.PerParticleDistanceScale:NiagaraFloat=&RemapRange(
     Input Value To Remap=&VectorLength(Vector=&SubtractVectorFromPosition(
       Position=&ReturnCameraPosition(Coordinate Space:ENiagaraCoordinateSpace=::Simulation, Camera Query:NiagaraDataInterfaceCamera=DataInterface())
       Vector=Linked() {linkedVariable {name=Particles.Position, type=NiagaraPosition}}
      ))
     Input Min=500
     Input Max=8000
     Output Min=0
     Output Max=1.5
    )) {enabled=false, isSetParameters=true}
   &ScaleSpriteSize(
    Uniform Curve Scale=2
    Uniform Curve Sprite Scale=DataInterface() {propertyValues {Curve {keys [
        {arriveTangent=5.6242, interpMode=RCIM_Cubic, leaveTangent=5.6242, tangentMode=RCTM_User, time=0, value=0}
        {arriveTangent=0.9995, interpMode=RCIM_Cubic, leaveTangent=0.9995, tangentMode=RCTM_User, time=0.5, value=1}
       ]}}}
   )
   /SetVariables(Particles.SpriteSize=&Multiply_Vector2D_ByFloat(
     Float=&FN_LOD_Distance_Remap(Start Distance=5000, End Distance=25000, Minimum Output=1, Maximum Output=3, Clamp Output=true)
     Vector2D=$Particles.SpriteSize
    )) {isSetParameters=true}
   &GravityForce(Gravity=&RandomRangeVector(Minimum=<0 0 2500>, Maximum=<0 0 2500>))
   &Drag(Drag=0, Use Linear Drag=true, Use Rotational Drag=false)
   &ScaleColor(
    Scale Alpha=&Multiply_Float(
     A=&Multiply_Float(
      A=&FloatFromCurve(FloatCurve=DataInterface() {propertyValues {Curve {keys [
           {arriveTangent=4.0891, interpMode=RCIM_Cubic, leaveTangent=4.0891, tangentMode=RCTM_User, time=0, value=0}
           {arriveTangent=-0.7491, interpMode=RCIM_Cubic, leaveTangent=-0.7491, tangentMode=RCTM_User, time=0.2195, value=0.6071}
           {arriveTangent=-0.7778, interpMode=RCIM_Cubic, leaveTangent=-0.7778, time=0.3529, value=0.3224}
           {arriveTangent=0.2036, interpMode=RCIM_Cubic, tangentMode=RCTM_Break, time=1, value=0}
          ]}}})
      B=&FloatFromCurve(
       FloatCurve=DataInterface() {propertyValues {Curve {keys [<at 0 0.5>,<at 0.0992 1>,<at 0.8829 1>,<at 1 0.5>]}}}
       CurveIndex=Linked() {linkedVariable {name=NPC.FortniteNPC.TimeOfDay_0-1, type=NiagaraFloat}}
      )
     )
     B=&FN_LOD_Distance_Remap(Start Distance=5000, End Distance=25000, Minimum Output=0, Maximum Output=0.5, Clamp Output=true)
    )
    ScaleA=true
    ScaleRGB=true
   )
   &VortexForce(Origin Pull Amount=2500, UseOriginOffset=true, Vortex Force Amount=2500, Vortex Origin Offset:Vector3f=<0 0 2500>)
   &SolveForcesAndVelocity(Clamp Velocity=true, Limit Acceleration=true, Speed Limit=1500)
   &DynamicMaterialParameters(
    Index 0 Param 1=&FloatFromCurve(FloatCurve=DataInterface() {propertyValues {Curve {keys [
         {interpMode=RCIM_Cubic, leaveTangent=1.2103, tangentMode=RCTM_Break, time=0, value=0}
         {arriveTangent=1.2545, tangentMode=RCTM_User, time=1, value=1}
        ]}}})
    Module.Index0Float0Write=true
    Module.Index0Float1Write=true
    Module.Index0Float2Write=true
    Module.Index0Float3Write=true
   )
  ]
  rendererClasses [&NiagaraSpriteRendererProperties]
  renderers [{rendererClass=&NiagaraSpriteRendererProperties}]
 }
 Drops_Outer_GPU002 {
  simTarget=GPUComputeSim
  emitterUpdateScript [
   &EmitterState()
   &SpawnRate(SpawnRate=&Multiply_Float(A=250, B=1), Use Spawn Probability=false)
  ]
  particleSpawnScript [
   &InitializeParticle(
    AdjustAlpha=false
    AdjustHue=false
    AdjustSaturation=true
    AdjustValue=true
    Color=$System.TOD_Color_Water
    Color Mode=::NewEnumerator3
    Hue Shift Range:Vector2f=<-0.1 1>
    Lifetime Mode=::NewEnumerator1
    Mesh Scale Mode=::NewEnumerator1
    Mesh Uniform Scale Max=0.2
    Mesh Uniform Scale Min=0.1
    Saturation Range:Vector2f=<1 1.4>
    UsePositionOffset=false
    Value Range:Vector2f=<1 2>
    Write Material Random=true
   )
   &ShapeLocation(
    Handle Radius=250
    Large Radius=2500
    MaskSpawnGroup=false
    Offset Mode=::NewEnumerator0
    Override Seed=false
    Shape Primitive=::NewEnumerator3
    UseOwnerScale=true
   ) {module=ShapeLocation001}
   &DynamicMaterialParameters(
    Index 0 Param 1=0.5
    Module.Index0Float0Write=true
    Module.Index0Float1Write=true
    Module.Index0Float2Write=true
    Module.Index0Float3Write=true
   )
   &AddVelocity(Velocity=<0 0 800>)
  ]
  particleUpdateScript [
   &ParticleState()
   /SetVariables(Particles.PerParticleDistanceScale:NiagaraFloat=&RemapRange(
     Input Value To Remap=&VectorLength(Vector=&SubtractVectorFromPosition(
       Position=&ReturnCameraPosition(Coordinate Space:ENiagaraCoordinateSpace=::Simulation, Camera Query:NiagaraDataInterfaceCamera=?)
       Vector=Linked() {linkedVariable {name=Particles.Position, type=NiagaraPosition}}
      ))
     Input Min=250
     Input Max=2000
     Output Min=0
     Output Max=1
    )) {isSetParameters=true}
   &ScaleMeshSize(Scale Factor=&VectorFromFloat(Value=&FloatFromCurve(
      FloatCurve=DataInterface() {propertyValues {Curve {keys [
          <at 0 0 RCIM_Cubic>
          {arriveTangent=0.8739, interpMode=RCIM_Cubic, leaveTangent=0.8739, time=0.5554, value=0.7175}
          {arriveTangent=-1.6139, interpMode=RCIM_Cubic, leaveTangent=-1.6139, time=0.8915, value=0.7791}
          <at 1 0 RCIM_Cubic>
         ]}}}
      Scale Curve=$Particles.PerParticleDistanceScale
     ))) {module=ScaleMeshSize001}
   &VortexForce(Origin Pull Amount=100, UseOriginOffset=false, Vortex Force Amount=500)
   &LinearForce(Force=<0 0 50>)
   &SolveForcesAndVelocity(Clamp Velocity=false, Limit Acceleration=false)
   &ScaleColor(
    Scale Alpha=&FloatFromCurve(FloatCurve=DataInterface() {propertyValues {Curve {keys [<at 0.54 1 RCIM_Cubic>, {arriveTangent=-6.1233, interpMode=RCIM_Cubic, leaveTangent=-6.1233, tangentMode=RCTM_User, time=1, value=0}]}}})
    Scale RGB=<25 0 10>
    ScaleA=true
    ScaleRGB=true
   )
  ]
  rendererClasses [&NiagaraMeshRendererProperties]
  renderers [{rendererClass=&NiagaraMeshRendererProperties, rendererIndex=0}]
 }
 Drops_Outer_GPU003 {
  simTarget=GPUComputeSim
  emitterUpdateScript [
   &EmitterState()
   &SpawnRate(SpawnRate=&Multiply_Float(A=600, B=$System.DistanceSpawnMult), Use Spawn Probability=false) {enabled=false}
   &SpawnParticlesFromOtherEmitter(
    Attribute Reader=DataInterface() {propertyValues {EmitterBinding {bindingMode=Other, emitterName=Drops_Outer_GPU002}}}
    Maximum Age Spawn Threshold=0.8
    Minimum Age Spawn Threshold=0
    Spawn Rate=3
    Spawn Rate Per Particle Cap=3
    Use Maximum Age=false
    Use Minimum Age=false
   )
  ]
  particleSpawnScript [
   &InitializeParticle(
    Color=<rgba 1 0.3351 0.8605 1>
    Lifetime Min=1.5
    Lifetime Mode=::NewEnumerator1
    Mesh Scale Mode=::NewEnumerator1
    Mesh Uniform Scale Max=0.1
    Mesh Uniform Scale Min=0.1
    UsePositionOffset=false
    Write Material Random=true
   )
   &ApplyInitialForces()
   &SampleParticlesFromOtherEmitter(
    Applied Position Offset:Vector3f=&RandomRangeVector(Minimum=<-15 -15 -15>, Maximum=<15 15 15>)
    Applied Velocity Scale:NiagaraFloat=&RandomRangeFloat(Minimum=0.15, Maximum=0.4)
    Velocity Application:ENiagara_DirectReadApplicationMode=::NewEnumerator0
    Velocity Sampling=::NewEnumerator0
   )
   &DynamicMaterialParameters(
    Index 0 Param 1=0.3
    Module.Index0Float0Write=true
    Module.Index0Float1Write=true
    Module.Index0Float2Write=true
    Module.Index0Float3Write=true
   )
  ]
  particleUpdateScript [
   &ParticleState()
   &ScaleMeshSize(Scale Factor=&VectorFromCurve(VectorCurve=DataInterface() {propertyValues {
       XCurve {keys [
         {arriveTangent=25.1764, interpMode=RCIM_Cubic, leaveTangent=25.1764, tangentMode=RCTM_User, time=0.004, value=0}
         {arriveTangent=-2.4305, interpMode=RCIM_Cubic, leaveTangent=-2.4305, tangentMode=RCTM_User, time=0.0237, value=1.2223}
         <at 1 0.5>
        ]}
       YCurve {keys [{arriveTangent=17.4554, interpMode=RCIM_Cubic, leaveTangent=17.4554, tangentMode=RCTM_User, time=0, value=0}, <at 0.1573 1.7422>, <at 1 0.5>]}
       ZCurve {keys [{arriveTangent=18.6187, interpMode=RCIM_Cubic, leaveTangent=18.6187, tangentMode=RCTM_User, time=0, value=0}, <at 0.1533 1.77>, <at 1 0.5>]}
      }})) {module=ScaleMeshSize001}
   &PointForce(Force Strength=-50, Module.Use Falloff=false, Module.UseLinearForce=false) {module=PointForce001}
   &SolveForcesAndVelocity(Clamp Velocity=false, Limit Acceleration=false)
   &ScaleColor(
    Scale Alpha=&FloatFromCurve(
     FloatCurve=DataInterface() {propertyValues {Curve {keys [
         {arriveTangent=11.3667, interpMode=RCIM_Cubic, leaveTangent=11.3667, tangentMode=RCTM_User, time=0, value=0}
         {arriveTangent=0.021, interpMode=RCIM_Cubic, leaveTangent=0.021, tangentMode=RCTM_User, time=0.25, value=1.0088}
         {arriveTangent=-4.0895, interpMode=RCIM_Cubic, leaveTangent=-4.0895, tangentMode=RCTM_User, time=1, value=0}
        ]}}}
     Scale Curve=0.1
    )
    Scale RGB=&VectorFromFloat(Value=&FloatFromCurve(FloatCurve=DataInterface() {propertyValues {Curve {keys [<at 0 1>,<at 0.4 1>,<at 1 1>]}}}))
    ScaleA=true
    ScaleRGB=true
   )
  ]
  rendererClasses [&NiagaraMeshRendererProperties]
  renderers [{rendererClass=&NiagaraMeshRendererProperties, rendererIndex=0}]
 }
 Empty {
  bEnabled=false
  simTarget=GPUComputeSim
  emitterUpdateScript [
   &EmitterState()
   &SpawnRate(SpawnRate=25, Use Spawn Probability=false)
  ]
  particleSpawnScript [
   &InitializeParticle(
    Color=&ColorFromCurve(
     ColorCurve=DataInterface() {propertyValues {
       AlphaCurve {keys [{interpMode=RCIM_Cubic, tangentMode=RCTM_User, time=-0.0057, value=1}, {interpMode=RCIM_Cubic, tangentMode=RCTM_User, time=1.0058, value=1}]}
       BlueCurve {keys [
         {interpMode=RCIM_Cubic, leaveTangentWeight=0.0473, tangentMode=RCTM_User, time=-0.1419, value=0.092}
         {
          arriveTangent=-0.0529
          arriveTangentWeight=0.0474
          interpMode=RCIM_Cubic
          leaveTangent=-0.0529
          leaveTangentWeight=0.047
          tangentMode=RCTM_User
          time=0
          value=0.1781
         }
         {
          arriveTangent=-0.0887
          arriveTangentWeight=0.0471
          interpMode=RCIM_Cubic
          leaveTangent=-0.0887
          leaveTangentWeight=0.0072
          tangentMode=RCTM_User
          time=0.1409
          value=0.1539
         }
         {
          arriveTangent=-1.1304
          arriveTangentWeight=0.0224
          interpMode=RCIM_Cubic
          leaveTangent=-1.1304
          leaveTangentWeight=0.0193
          tangentMode=RCTM_User
          time=0.1623
          value=0.1367
         }
         {
          arriveTangent=-0.664
          arriveTangentWeight=0.0074
          interpMode=RCIM_Cubic
          leaveTangent=-0.664
          leaveTangentWeight=0.0094
          tangentMode=RCTM_User
          time=0.1809
          value=0.1087
         }
         {
          arriveTangent=0.6098
          arriveTangentWeight=0.0092
          interpMode=RCIM_Cubic
          leaveTangent=0.6098
          leaveTangentWeight=0.0028
          tangentMode=RCTM_User
          time=0.2044
          value=0.1087
         }
         {
          arriveTangent=2.3315
          arriveTangentWeight=0.0061
          interpMode=RCIM_Cubic
          leaveTangent=2.3315
          leaveTangentWeight=0.0366
          tangentMode=RCTM_User
          time=0.2116
          value=0.1275
         }
         {
          arriveTangent=1.1205
          arriveTangentWeight=0.0217
          interpMode=RCIM_Cubic
          leaveTangent=1.1205
          leaveTangentWeight=0.0372
          tangentMode=RCTM_User
          time=0.2549
          value=0.2264
         }
         {
          arriveTangent=0.2237
          arriveTangentWeight=0.0253
          interpMode=RCIM_Cubic
          leaveTangent=0.2237
          leaveTangentWeight=0.0648
          tangentMode=RCTM_User
          time=0.3291
          value=0.2591
         }
         {
          arriveTangent=0.1845
          arriveTangentWeight=0.0643
          interpMode=RCIM_Cubic
          leaveTangent=0.1845
          leaveTangentWeight=0.0381
          tangentMode=RCTM_User
          time=0.5187
          value=0.2854
         }
         {
          arriveTangent=0.1461
          arriveTangentWeight=0.0379
          interpMode=RCIM_Cubic
          leaveTangent=0.1461
          leaveTangentWeight=0.026
          tangentMode=RCTM_User
          time=0.6311
          value=0.3149
         }
         {
          arriveTangent=-0.9326
          arriveTangentWeight=0.0331
          interpMode=RCIM_Cubic
          leaveTangent=-0.9326
          leaveTangentWeight=0.0309
          tangentMode=RCTM_User
          time=0.7083
          value=0.3131
         }
         {
          arriveTangent=-1.0096
          arriveTangentWeight=0.0428
          interpMode=RCIM_Cubic
          leaveTangent=-1.0096
          leaveTangentWeight=0.0513
          tangentMode=RCTM_User
          time=0.7805
          value=0.1755
         }
         {
          arriveTangent=-0.4797
          arriveTangentWeight=0.0429
          interpMode=RCIM_Cubic
          leaveTangent=-0.4797
          leaveTangentWeight=0.0155
          tangentMode=RCTM_User
          time=0.8672
          value=0.1527
         }
         {
          arriveTangent=-0.5158
          arriveTangentWeight=0.0232
          interpMode=RCIM_Cubic
          leaveTangent=-0.5158
          leaveTangentWeight=0.0137
          tangentMode=RCTM_User
          time=0.8985
          value=0.1189
         }
         {
          arriveTangent=0.5759
          arriveTangentWeight=0.0086
          interpMode=RCIM_Cubic
          leaveTangent=0.5759
          leaveTangentWeight=0.0134
          tangentMode=RCTM_User
          time=0.917
          value=0.127
         }
         {
          arriveTangent=0.4127
          arriveTangentWeight=0.0121
          interpMode=RCIM_Cubic
          leaveTangent=0.4127
          leaveTangentWeight=0.014
          tangentMode=RCTM_User
          time=0.9458
          value=0.1461
         }
         {
          arriveTangent=0.5904
          arriveTangentWeight=0.014
          interpMode=RCIM_Cubic
          leaveTangent=0.5904
          leaveTangentWeight=0.0087
          tangentMode=RCTM_User
          time=0.9792
          value=0.1527
         }
         {arriveTangentWeight=0.0069, interpMode=RCIM_Cubic, tangentMode=RCTM_User, time=1, value=0.1781}
        ]}
       GreenCurve {keys [
         {interpMode=RCIM_Cubic, leaveTangentWeight=0.0473, tangentMode=RCTM_User, time=-0.1419, value=0.092}
         {
          arriveTangent=-0.0111
          arriveTangentWeight=0.0473
          interpMode=RCIM_Cubic
          leaveTangent=-0.0111
          leaveTangentWeight=0.047
          tangentMode=RCTM_User
          time=0
          value=0.1348
         }
         {
          arriveTangent=-0.1572
          arriveTangentWeight=0.0549
          interpMode=RCIM_Cubic
          leaveTangent=-0.1572
          leaveTangentWeight=0.0084
          tangentMode=RCTM_User
          time=0.1409
          value=0.1163
         }
         {
          arriveTangent=-0.5174
          arriveTangentWeight=0.0081
          interpMode=RCIM_Cubic
          leaveTangent=-0.5174
          leaveTangentWeight=0.007
          tangentMode=RCTM_User
          time=0.1623
          value=0.1093
         }
         {
          arriveTangent=0.2567
          arriveTangentWeight=0.0064
          interpMode=RCIM_Cubic
          leaveTangent=0.2567
          leaveTangentWeight=0.0081
          tangentMode=RCTM_User
          time=0.1809
          value=0.0985
         }
         {
          arriveTangent=1.5441
          arriveTangentWeight=0.0145
          interpMode=RCIM_Cubic
          leaveTangent=1.5441
          leaveTangentWeight=0.0044
          tangentMode=RCTM_User
          time=0.2044
          value=0.1343
         }
         {
          arriveTangent=0.6936
          arriveTangentWeight=0.0029
          interpMode=RCIM_Cubic
          leaveTangent=0.6936
          leaveTangentWeight=0.0175
          tangentMode=RCTM_User
          time=0.2116
          value=0.1461
         }
         {
          arriveTangent=0.4378
          arriveTangentWeight=0.0157
          interpMode=RCIM_Cubic
          leaveTangent=0.4378
          leaveTangentWeight=0.0266
          tangentMode=RCTM_User
          time=0.2549
          value=0.1693
         }
         {
          arriveTangent=0.1985
          arriveTangentWeight=0.0249
          interpMode=RCIM_Cubic
          leaveTangent=0.1985
          leaveTangentWeight=0.0648
          tangentMode=RCTM_User
          time=0.3281
          value=0.197
         }
         {
          arriveTangent=0.1815
          arriveTangentWeight=0.0646
          interpMode=RCIM_Cubic
          leaveTangent=0.1815
          leaveTangentWeight=0.0381
          tangentMode=RCTM_User
          time=0.5187
          value=0.2217
         }
         {
          arriveTangent=0.1637
          arriveTangentWeight=0.038
          interpMode=RCIM_Cubic
          leaveTangent=0.1637
          leaveTangentWeight=0.0261
          tangentMode=RCTM_User
          time=0.6311
          value=0.2521
         }
         {
          arriveTangent=-0.5568
          arriveTangentWeight=0.028
          interpMode=RCIM_Cubic
          leaveTangent=-0.5568
          leaveTangentWeight=0.0262
          tangentMode=RCTM_User
          time=0.7083
          value=0.2527
         }
         {
          arriveTangent=-0.4339
          arriveTangentWeight=0.0323
          interpMode=RCIM_Cubic
          leaveTangent=-0.4339
          leaveTangentWeight=0.0387
          tangentMode=RCTM_User
          time=0.7805
          value=0.1689
         }
         {
          arriveTangent=-0.2899
          arriveTangentWeight=0.039
          interpMode=RCIM_Cubic
          leaveTangent=-0.2899
          leaveTangentWeight=0.0141
          tangentMode=RCTM_User
          time=0.8672
          value=0.1838
         }
         {
          arriveTangent=-1.2376
          arriveTangentWeight=0.0302
          interpMode=RCIM_Cubic
          leaveTangent=-1.2376
          leaveTangentWeight=0.0178
          tangentMode=RCTM_User
          time=0.8985
          value=0.1347
         }
         {
          arriveTangent=0.0037
          arriveTangentWeight=0.0113
          interpMode=RCIM_Cubic
          leaveTangent=0.0037
          leaveTangentWeight=0.0176
          tangentMode=RCTM_User
          time=0.917
          value=0.1221
         }
         {
          arriveTangent=0.2047
          arriveTangentWeight=0.0134
          interpMode=RCIM_Cubic
          leaveTangent=0.2047
          leaveTangentWeight=0.0155
          tangentMode=RCTM_User
          time=0.9458
          value=0.1348
         }
         {
          arriveTangent=-0.0002
          arriveTangentWeight=0.0187
          interpMode=RCIM_Cubic
          leaveTangent=-0.0002
          leaveTangentWeight=0.0117
          tangentMode=RCTM_User
          time=0.9792
          value=0.1348
         }
         {arriveTangentWeight=0.0069, interpMode=RCIM_Cubic, tangentMode=RCTM_User, time=1, value=0.1348}
        ]}
       RedCurve {keys [
         {interpMode=RCIM_Cubic, leaveTangentWeight=0.0473, tangentMode=RCTM_User, time=-0.1419, value=0.092}
         {
          arriveTangent=0.0279
          arriveTangentWeight=0.0485
          interpMode=RCIM_Cubic
          leaveTangent=0.0279
          leaveTangentWeight=0.0482
          tangentMode=RCTM_User
          time=0
          value=0.1119
         }
         {
          arriveTangent=0.0066
          arriveTangentWeight=0.0514
          interpMode=RCIM_Cubic
          leaveTangent=0.0066
          leaveTangentWeight=0.0078
          tangentMode=RCTM_User
          time=0.1409
          value=0.0999
         }
         {
          arriveTangent=0.8688
          arriveTangentWeight=0.0095
          interpMode=RCIM_Cubic
          leaveTangent=0.8688
          leaveTangentWeight=0.0082
          tangentMode=RCTM_User
          time=0.1623
          value=0.113
         }
         {
          arriveTangent=2.2289
          arriveTangentWeight=0.0151
          interpMode=RCIM_Cubic
          leaveTangent=2.2289
          leaveTangentWeight=0.0192
          tangentMode=RCTM_User
          time=0.1809
          value=0.1422
         }
         {
          arriveTangent=3.0926
          arriveTangentWeight=0.0256
          interpMode=RCIM_Cubic
          leaveTangent=3.0926
          leaveTangentWeight=0.0078
          tangentMode=RCTM_User
          time=0.2044
          value=0.2412
         }
         {
          arriveTangent=-1.4403
          arriveTangentWeight=0.0042
          interpMode=RCIM_Cubic
          leaveTangent=-1.4403
          leaveTangentWeight=0.0253
          tangentMode=RCTM_User
          time=0.2116
          value=0.2374
         }
         {
          arriveTangent=-0.4667
          arriveTangentWeight=0.0159
          interpMode=RCIM_Cubic
          leaveTangent=-0.4667
          leaveTangentWeight=0.0269
          tangentMode=RCTM_User
          time=0.2549
          value=0.1685
         }
         {
          arriveTangent=0.1084
          arriveTangentWeight=0.0245
          interpMode=RCIM_Cubic
          leaveTangent=0.1084
          leaveTangentWeight=0.0639
          tangentMode=RCTM_User
          time=0.3281
          value=0.1831
         }
         {
          arriveTangent=0.1435
          arriveTangentWeight=0.0642
          interpMode=RCIM_Cubic
          leaveTangent=0.1435
          leaveTangentWeight=0.0378
          tangentMode=RCTM_User
          time=0.5187
          value=0.1972
         }
         {
          arriveTangent=0.1553
          arriveTangentWeight=0.0379
          interpMode=RCIM_Cubic
          leaveTangent=0.1553
          leaveTangentWeight=0.026
          tangentMode=RCTM_User
          time=0.6311
          value=0.2266
         }
         {
          arriveTangent=-0.4817
          arriveTangentWeight=0.0273
          interpMode=RCIM_Cubic
          leaveTangent=-0.4817
          leaveTangentWeight=0.0255
          tangentMode=RCTM_User
          time=0.7083
          value=0.2266
         }
         {
          arriveTangent=-0.2068
          arriveTangentWeight=0.0242
          interpMode=RCIM_Cubic
          leaveTangent=-0.2068
          leaveTangentWeight=0.029
          tangentMode=RCTM_User
          time=0.7805
          value=0.1546
         }
         {
          arriveTangent=0.4933
          arriveTangentWeight=0.0291
          interpMode=RCIM_Cubic
          leaveTangent=0.4933
          leaveTangentWeight=0.0105
          tangentMode=RCTM_User
          time=0.8672
          value=0.1937
         }
         {
          arriveTangent=-0.9567
          arriveTangentWeight=0.0274
          interpMode=RCIM_Cubic
          leaveTangent=-0.9567
          leaveTangentWeight=0.0161
          tangentMode=RCTM_User
          time=0.8985
          value=0.2128
         }
         {
          arriveTangent=-1.757
          arriveTangentWeight=0.0213
          interpMode=RCIM_Cubic
          leaveTangent=-1.757
          leaveTangentWeight=0.0331
          tangentMode=RCTM_User
          time=0.917
          value=0.1461
         }
         {
          arriveTangent=-0.5697
          arriveTangentWeight=0.0193
          interpMode=RCIM_Cubic
          leaveTangent=-0.5697
          leaveTangentWeight=0.0224
          tangentMode=RCTM_User
          time=0.9458
          value=0.1298
         }
         {
          arriveTangent=-0.3306
          arriveTangentWeight=0.0218
          interpMode=RCIM_Cubic
          leaveTangent=-0.3306
          leaveTangentWeight=0.0136
          tangentMode=RCTM_User
          time=0.9792
          value=0.1106
         }
         {arriveTangentWeight=0.0069, interpMode=RCIM_Cubic, tangentMode=RCTM_User, time=1, value=0.1119}
        ]}
      }}
     CurveIndex=Linked() {linkedVariable {name=NPC.FortniteNPC.TimeOfDay_0-1, type=NiagaraFloat}}
    )
    Lifetime=&RandomRangeFloat(Minimum=2, Maximum=2.5)
    Mesh Scale Mode=::NewEnumerator1
    Mesh Uniform Scale Max=12
    Mesh Uniform Scale Min=2
    Sprite Rotation Angle=&RandomRangeFloat(Minimum=-0.15, Maximum=0.15)
    Sprite Rotation Mode=::NewEnumerator1
    Sprite Size Mode=::NewEnumerator1
    Sprite UV Mode=::NewEnumerator1
    Uniform Sprite Size Max=2500
    Uniform Sprite Size Min=1500
    UsePositionOffset=false
    Write Material Random=true
   )
   &ShapeLocation(
    Large Radius=4000
    MaskSpawnGroup=false
    Non Uniform Scale=<1 1 10>
    Offset=<0 0 25>
    Offset Mode=::NewEnumerator0
    Override Seed=false
    Shape Primitive=::NewEnumerator3
    UseOwnerScale=true
   )
   /Niagara/Modules/Landscape/SampleLandscape(Height Offset=250)
   &AddVelocity(Velocity=<0 0 150>)
  ]
  particleUpdateScript [
   &ParticleState()
   /Niagara/Modules/Update/Camera/ViewMask(Border Percent Max=80, Border Percent Min=50) {enabled=false}
   &GravityForce(Gravity=<0 0 1500>)
   &VortexForce(Origin Pull Amount=2500, UseOriginOffset=true, Vortex Force Amount=2500, Vortex Origin Offset:Vector3f=<0 0 2500>)
   &ScaleSpriteSize(Uniform Curve Sprite Scale=DataInterface() {propertyValues {Curve {keys [{interpMode=RCIM_Cubic, tangentMode=RCTM_Break, time=0, value=0.5}, {tangentMode=RCTM_User, time=1, value=1}]}}})
   &ScaleColor(
    Scale Alpha=&FloatFromCurve(
     FloatCurve=DataInterface() {propertyValues {Curve {keys [
         {interpMode=RCIM_Cubic, tangentMode=RCTM_User, time=0, value=0}
         {interpMode=RCIM_Cubic, tangentMode=RCTM_User, time=0.5, value=1}
         {tangentMode=RCTM_User, time=1, value=0}
        ]}}}
     Scale Curve=0.01
    )
    Scale RGB=<25 25 25>
    ScaleA=true
    ScaleRGB=true
   )
   &AccelerationForce(Acceleration=<150 0 25>)
   &Drag(Use Linear Drag=true, Use Rotational Drag=true)
   &SolveForcesAndVelocity(Clamp Velocity=true, Limit Acceleration=false)
   &SubUVAnimation(
    Random Start Frame=true
    "Sprite Renderer "=DataInterface() {propertyValues {SpriteRenderer=&self:Empty.NiagaraSpriteRendererProperties_2}}
    Start Frame Offset:NiagaraInt32=&RandomRangeInt(Minimum=0, Maximum=16)
    SubUV Lookup Index Scale:NiagaraFloat=&RandomRangeFloat(Minimum=1, Maximum=2)
    UseEndFrame=false
    UseStartFrame=false
   )
   &DynamicMaterialParameters(
    Index 0 Param 1=&FloatFromCurve(FloatCurve=DataInterface() {propertyValues {Curve {keys [
         {interpMode=RCIM_Cubic, tangentMode=RCTM_User, time=0, value=0.1}
         {arriveTangent=2.7279, interpMode=RCIM_Cubic, leaveTangent=2.7279, tangentMode=RCTM_User, time=1, value=1}
        ]}}})
    Index 0 Param 2=0.2
    Module.Index0Float0Write=true
    Module.Index0Float1Write=true
    Module.Index0Float2Write=true
    Module.Index0Float3Write=true
   )
  ]
  rendererClasses [&NiagaraSpriteRendererProperties]
  renderers [{rendererClass=&NiagaraSpriteRendererProperties}]
 }
 Onground {
  simTarget=GPUComputeSim
  emitterUpdateScript [
   &EmitterState()
   &SpawnRate(SpawnRate=2500, Use Spawn Probability=false)
  ]
  particleSpawnScript [
   &InitializeParticle(
    Color Maximum=<rgba 0.5693 0.0009 1 1>
    Color Minimum=<rgba 1 0.4948 0.8052 1>
    Color Mode=::NewEnumerator2
    Lifetime=&UniformRangedFloat(Minimum=3, Maximum=5)
    Mass Mode=::NewEnumerator1
    Position=&SimulationPosition()
    Position Mode=::NewEnumerator1
    Sprite Size Mode=::NewEnumerator1
    Uniform Sprite Size Max=20
    UsePositionOffset=false
    Write Material Random=true
   ) {module=InitializeParticle001}
   &AddVelocity(
    Cone Angle=90
    Cone Axis=<0 0 1>
    Module.UseConeFalloff=false
    Module.UseVelDistribution=false
    Override Seed=false
    Velocity Mode=::NewEnumerator2
    Velocity Speed=&RandomRangeFloat(Minimum=1, Maximum=250)
   )
   &InitialMeshOrientation(Mesh Axis To Orient=<0 0 1>, Mesh Orientation Mode=::NewEnumerator2, Orientation Vector=<0 0 1>, Use Rotation Vector=false)
   &ShapeLocation(
    Cylinder Height=500
    Cylinder Radius=5000
    MaskSpawnGroup=false
    Offset=<0 0 250>
    Offset Mode=::NewEnumerator0
    Override Seed=false
    Shape Primitive=::NewEnumerator1
    Surface Only=false
    UseOwnerScale=true
   )
   &SampleLandscape(
    Facing Axis:Vector3f=<0 0 1>
    Height Filter=true
    Height Max=2500
    Height Min=10
    Height Offset:NiagaraFloat=&RandomRangeFloat(Minimum=0, Maximum=250)
    Normal Filter=true
    Normal Filter Tolerance=0.2
    Orient Mesh To Normal=true
    Use Max Height=true
    Use Min Height=true
   )
   &ShapeLocation(
    Handle Radius=1500
    Large Radius=4000
    MaskSpawnGroup=false
    Non Uniform Scale=<1 1 0.25>
    Override Seed=false
    Shape Primitive=::NewEnumerator3
    UseOwnerScale=true
   ) {module=ShapeLocation001}
  ]
  particleUpdateScript [
   &ParticleState()
   &RegionMask(Outer Float=0.25, Region Falloff=1, Region Radius=5000, UseOffset=false, UseRotation=false)
   &ScaleSpriteSize(
    Uniform Curve Scale=&Multiply_Float(A=1, B=&Lerp_Float(A=0.25, B=1, Alpha=$Particles.RegionMask.Float))
    Uniform Curve Sprite Scale=DataInterface() {propertyValues {Curve {keys [{interpMode=RCIM_Cubic, tangentMode=RCTM_User, time=0, value=0}, <at 0.6814 0.8961 RCIM_Cubic>, {tangentMode=RCTM_User, time=1, value=0}]}}}
   )
   &ScaleColor(
    Scale Alpha=&FloatFromCurve(
     FloatCurve=DataInterface() {propertyValues {Curve {keys [
         {interpMode=RCIM_Cubic, tangentMode=RCTM_User, time=0, value=0}
         {interpMode=RCIM_Cubic, tangentMode=RCTM_User, time=0.5, value=1}
         {tangentMode=RCTM_User, time=1, value=0}
        ]}}}
     Scale Curve=&Multiply_Float(A=1, B=&OneMinusFloat(Float=$Output.ViewMask.ViewMask))
    )
    ScaleA=true
    ScaleRGB=true
   )
   /Niagara/Modules/Update/Forces/CurlNoiseForce(
    Noise Frequency:NiagaraFloat=&RandomRangeFloat(Minimum=0.002, Maximum=0.1)
    Noise Strength=&RandomRangeFloat(Minimum=250, Maximum=2500)
    Pan Noise=true
    Pan Noise Field=<0 1 1>
   )
   &GravityForce(Gravity=&Multiply_VectorByFloat(Vector=<0 0 4000>, Float=$Particles.RegionMask.Float))
   &VortexForce(Origin Pull Amount=-5000, UseOriginOffset=true, Vortex Force Amount=5000, Vortex Origin Offset:Vector3f=<0 0 5000>)
   &Drag(Drag=0, Use Linear Drag=true, Use Rotational Drag=true)
   &SolveForcesAndVelocity(
    Clamp Velocity=true
    Limit Acceleration=false
    Speed Limit:NiagaraFloat=&Multiply_Float(A=5000, B=&Power(a=$Particles.RegionMask.Float, b=4))
   )
   /SetVariables(Particles.NewSize:Vector2f=&Multiply_Vector2D_ByFloat(
     Float=&FN_LOD_Distance_Remap(Start Distance=2500, End Distance=20000, Minimum Output=0.5, Maximum Output=0, Clamp Output=true)
     Vector2D=&Multiply_Vector2D_ByFloat(Float=5, Vector2D=$Particles.SpriteSize)
    )) {isSetParameters=true}
   /SetVariables(Particles.Vector 2D:Vector2f=&Multiply_Vector2D(A=$Particles.NewSize, B=<1 1>)) {isSetParameters=true}
   /SetVariables(
    Particles.Linear Color=&MultiplyLinearColor(Linear Color=$Particles.Color, Scale Factor=<25 1 25 0.05>)
    Particles.Linear Color001:LinearColor=&MultiplyLinearColor(Linear Color=$Particles.Color, Scale Factor=<250 1 250 1>)
   ) {isSetParameters=true}
   &CameraOffset(Camera Offset Amount=1500)
   /SetVariables(Particles.Vector 2D001:Vector2f=&Multiply_Vector2D(A=$Particles.NewSize, B=<0.5 0.5>)) {isSetParameters=true}
  ]
  rendererClasses [&NiagaraSpriteRendererProperties]
  renderers [
   {rendererClass=&NiagaraSpriteRendererProperties}
   {rendererClass=&NiagaraSpriteRendererProperties, rendererIndex=1}
   {rendererClass=&NiagaraSpriteRendererProperties, rendererIndex=2}
  ]
 }
}
summary {
 systemName=NS_ZeroPoint_Updraft_Area
 emitters [
  {emitterName=Drops_Outer_GPU002, simTarget=GPUComputeSim, rendererClasses [&NiagaraMeshRendererProperties]}
  {emitterName=Drops_Outer_GPU003, simTarget=GPUComputeSim, rendererClasses [&NiagaraMeshRendererProperties]}
  {emitterName=BlowingSandFlipbook, simTarget=GPUComputeSim, rendererClasses [&NiagaraSpriteRendererProperties]}
  {emitterName=Onground, simTarget=GPUComputeSim, rendererClasses [&NiagaraSpriteRendererProperties]}
  {bEnabled=false, emitterName=Empty, simTarget=GPUComputeSim, rendererClasses [&NiagaraSpriteRendererProperties]}
 ]
 userVariables [User.Radius {type=NiagaraFloat, defaultValue=3500, description=""}]
}
